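# Sound cue for the batteries audio of Battle Pass Season 26 quest P4 Q07, routed through the Fort_Quests sound class.
# A Delay node with equal min and max holds playback for a fixed, sub-second pause before a single Wave Player plays the Batteries wave.
# The cue boosts volume and reuses the supply drop HiddenStar ambient loop attenuation asset for spatial falloff.
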SoundCue /BattlePassS26_Quests/Sound/Quest_S26_P04_Q07/Quest_S26_P4_Q07_Batteries_Cue
root=Delay_0
nodes [
 {id=Delay_0, children [WavePlayer_0], properties {delayMax=0.9, delayMin=0.9}}
 {id=WavePlayer_0, properties {soundWaveAssetPtr=&/Quest_S26_P4_Q07_Batteries}}
]
properties {
 attenuationSettings=/Game/Athena/Sounds/SupplyDrops/Fort_HiddenStar_Ambient_Loop_Att
 maxDistance=2500
 soundClassObject=&Fort_Quests
 volumeMultiplier=2.5
}
